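; Gameplay cue notify for the CherrySmook dance floor launch pad that plays a burst sound on the target.
; It overrides HandleGameplayCue, calls the parent, then casts the target to FortPlayerPawn.
; Whether that pawn is locally viewed selects between the 1P and 3P burst sounds, and the chosen sound spawns attached to the target's root component.
; The override does not branch on event type.
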(fn HandleGameplayCue
  (params (MyTarget :type Actor%20Object%20Reference) (EventType :type EGameplayCueEvent%20Enum :default "OnActive") (Parameters :type Gameplay%20Cue%20Parameters%20Structure%20%28by%20ref%29))
  (|CallParentFunction MyTarget EventType Parameters)
  (bind _asfort_player_pawn (Utilities|Casting|CastToFortPlayerPawn MyTarget))
  (Audio|SpawnSoundAttached (select (Pawn|IsLocallyViewed _asfort_player_pawn) (Variables|Default|Get1P%20Burst) (Variables|Default|Get3P%20Burst)) (Variables|/Script/Engine.Actor|GetRootComponent MyTarget))
  (return))
